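# Sound cue for the ambient rattle of a Jigsaw quest in Figment S05, using quest-interactable hint attenuation.
# A random node picks one of three equally weighted rattle waves and feeds it through a modulator at the root for pitch and volume variation.
# Concurrency overrides allow one instance at a time and block new plays, and the cue routes to the UI sound class.
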SoundCue /Figment_S05/Athena/Sounds/Quests/Jigsaw/Quest_Jigsaw_AmbRattle_Cue
root=Modulator_0
nodes [
 {id=Modulator_0, children [Random_1]}
 {id=Random_1, children [WavePlayer_1,WavePlayer_5,WavePlayer_6], properties {weights [1,1,1]}}
 {id=WavePlayer_1, properties {soundWaveAssetPtr=&/Quest_Jigsaw_AmbRattle_03}}
 {id=WavePlayer_5, properties {soundWaveAssetPtr=&/Quest_Jigsaw_AmbRattle_01}}
 {id=WavePlayer_6, properties {soundWaveAssetPtr=&/Quest_Jigsaw_AmbRattle_02}}
]
properties {
 attenuationSettings=/Game/Sounds/Attenuation/Fort_QuestInteractables_Hint
 maxDistance=1300
 soundClassObject=&Fort_UI
 volumeMultiplier=1
 concurrencyOverrides {maxCount=1, resolutionRule=PreventNew}
}
