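# Sound cue for collecting a jigsaw piece in a Figment Season 5 quest, in the Fort_UI sound class.
# A single collect wave passes through a default-setting Modulator, then a fixed short Delay at the root holds back playback.
# It uses quest-interactable hint attenuation and a boosted volume, with concurrency override values specifying one instance and preventing new ones.
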SoundCue /Figment_S05/Athena/Sounds/Quests/Jigsaw/Quest_Jigsaw_Collect_Cue
root=Delay_0
nodes [
 {id=Delay_0, children [Modulator_0], properties {delayMax=0.25, delayMin=0.25}}
 {id=Modulator_0, children [WavePlayer_1]}
 {id=WavePlayer_1, properties {soundWaveAssetPtr=&/Quest_Jigsaw_Collect_01}}
]
properties {
 attenuationSettings=/Game/Sounds/Attenuation/Fort_QuestInteractables_Hint
 maxDistance=1300
 soundClassObject=&Fort_UI
 volumeMultiplier=1.6
 concurrencyOverrides {maxCount=1, resolutionRule=PreventNew}
}
